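# Niagara candy-burst effect for the SoapPocket Punch emote, combining candy, confetti and short impact accents.
# GPU candy randomly varies by visibility tag across mesh and sprite renderers, bouncing through collision modules gated by quality level and front-end preview.
# A GPU confetti emitter packs three spawn groups with separate sprite colors and sizes.
# Flash, ripple, streak and light emitters add the hit, several tinted or dimmed by time of day.
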NiagaraSystem /BRCosmetics/Animation/Game/MainPlayer/Emotes/SoapPocket_Punch/FX/NS_Emotes_SoapPocketPunch_Candy
dependencies {
 usedDataInterfaces [
  &NiagaraDataInterfaceArrayInt32
  &NiagaraDataInterfaceArrayColor
  &NiagaraDataInterfaceMeshRendererInfo
  &NiagaraDataInterfaceSpriteRendererInfo
  &NiagaraDataInterfacePlatformSet
  &NiagaraDataInterfaceCurve
  &NiagaraDataInterfaceColorCurve
  &NiagaraDataInterfaceVector2DCurve
 ]
 usedDynamicInputs []
 usedModules [&SystemState]
}
emitters {
 Candy {
  simTarget=GPUComputeSim
  emitterUpdateScript [
   &EmitterState(Life Cycle Mode=::NewEnumerator1, Loop Behavior=::NewEnumerator1, Loop Duration=4, UseLoopDelay=false)
   &SpawnBurst_Instantaneous(Spawn Count=50, Use Spawn Probability=false, UseLoopCountLimit=false)
  ]
  particleSpawnScript [
   /SetVariables(Particles.VisibilityTag=&SelectIntFromArray(
     Direct Array Index=&RandomRangeInt(Minimum=0, Maximum=2)
     Int Selection Array=DataInterface() {propertyValues {IntData [0,1,1]}}
    )) {isSetParameters=true}
   &InitializeParticle(
    Color=&MaskLinearColor(
     Mask Parameter=$Particles.VisibilityTag
     Mask=1
     Masked Linear Color=&SelectLinearColorFromArray(
      Array Sampling Mode=::NewEnumerator0
      Direct Array Index=?
      Color Selection Array=DataInterface() {propertyValues {ColorData [<rgba 0.2218 1 0.9828 1>, <rgba 1 1 1 1>]}}
      Random Seed=0
     )
     Passthrough Linear Color=<rgba 1 1 1 1>
    )
    Lifetime Min=2.5
    Lifetime Mode=::NewEnumerator1
    Mesh Renderer Array Visibility Mode=::NewEnumerator3
    Mesh Renderer Info=DataInterface() {propertyValues {MeshRenderer=&self:Candy_0.NiagaraMeshRendererProperties_0}}
    Sprite Rotation Mode=::NewEnumerator2
    Sprite Size Mode=::NewEnumerator3
    UsePositionOffset=false
    Write Material Random=true
   )
   &PointForce(
    Force Origin=&SubtractVectorFromPosition(Position=/Niagara/DynamicInputs/Helpers/SimulationPosition(), Vector=<0 0 0>)
    Force Strength=&RandomRangeFloat(Minimum=300, Maximum=100)
    Linear Force Vector:Vector3f=&TransformVector(Vector=&RandomRangeVector(Minimum=<-1 -0.4 1>, Maximum=<1 1 1>), Destination Space=::World)
    Module.Use Falloff=false
    Module.UseLinearForce=true
   )
   &ShapeLocation(MaskSpawnGroup=false, Override Seed=false, Sphere Radius=10, Sphere Surface Distribution=1, UseOwnerScale=true)
   &InitialMeshOrientation(Use Rotation Vector=false)
   &SubUVAnimation(
    "Sprite Renderer "=DataInterface() {propertyValues {SpriteRenderer=&self:Candy_0.NiagaraSpriteRendererProperties_1}}
    SubUV Animation Mode:ENiagara_SubUVLookupModeV2=::NewEnumerator2
    UseEndFrame=false
    UseStartFrame=false
   )
   &AddVelocity(
    Module.Use Falloff Distance=false
    Origin Offset=<0 -4 -5>
    Override Seed=false
    Velocity Mode=::NewEnumerator1
    Velocity Speed=&RandomRangeFloat(Minimum=250, Maximum=150)
   ) {enabled=false}
   &ApplyInitialForces()
  ]
  particleUpdateScript [
   &ParticleState()
   &GravityForce()
   &Collision(
    Age Colliding Particles=false
    Collision Enabled=&Boolean_AndOperation(
     Bool1=&IsPlatformSetActive(New Platform Set=DataInterface() {propertyValues {Platforms {qualityLevelMask=-8}}})
     Bool2=&Boolean_NotOperation(Bool=$User.bIsFrontEndPreview)
    )
    Control Roll On Collision=true
    Enable Rest State=false
    Friction=0.1
    Kill Occluded Particles=false
    Kill Offscreen Particles=true
    KillOnCollision=false
    Module.EnableMaxCollisionCount=false
    Module.Randomize Collision Normal Vector=false
    Particle Radius Scale=5
    Radius Calculation Type=::NewEnumerator1
    Restitution=0.4
    ScaleRateOfRotation=true
    Z Depth Query Type:SceneDepthCollisionQueryMethod=::NewEnumerator2
   )
   &Collision(
    Age Colliding Particles=false
    Collision Enabled=&IsPlatformSetActive(New Platform Set=DataInterface() {propertyValues {Platforms {qualityLevelMask=-28}}})
    Control Roll On Collision=true
    Enable Rest State=false
    Friction=0.1
    Kill Occluded Particles=false
    Kill Offscreen Particles=true
    KillOnCollision=false
    Module.EnableMaxCollisionCount=false
    Module.Randomize Collision Normal Vector=false
    Particle Radius Scale=5
    Radius Calculation Type=::NewEnumerator1
    Restitution=0.8
    ScaleRateOfRotation=false
    Z Depth Query Type:SceneDepthCollisionQueryMethod=::NewEnumerator2
   ) {enabled=false, module=Collision001}
   &Collision(
    Age Colliding Particles=false
    Collision Enabled=&Boolean_OrOperation(
     Bool1=&IsPlatformSetActive(New Platform Set=DataInterface() {propertyValues {Platforms {qualityLevelMask=-28}}})
     Bool2=$User.bIsFrontEndPreview
    )
    Control Roll On Collision=true
    Enable Rest State=false
    Friction=0.1
    Kill Occluded Particles=false
    Kill Offscreen Particles=true
    KillOnCollision=false
    Module.EnableMaxCollisionCount=false
    Module.Randomize Collision Normal Vector=false
    Particle Radius Scale=5
    Radius Calculation Type=::NewEnumerator1
    Restitution=0.8
    ScaleRateOfRotation=false
    Z Depth Query Type:SceneDepthCollisionQueryMethod=::NewEnumerator2
   ) {module=Collision002}
   &KillParticlesInVolume(Box Size:Vector3f=<25 25 100>, Kill Shape=::NewEnumerator1, Origin Offset:Vector3f=<0 0 -50>) {enabled=false}
   &ScaleMeshSize(Scale Factor=&Multiply_VectorByFloat(
     Vector=<1 1 1>
     Float=&FloatFromCurve(FloatCurve=DataInterface() {propertyValues {Curve {keys [
          {tangentMode=RCTM_User, time=0, value=1}
          <at 0.9036 1>
          {arriveTangent=-0.0405, interpMode=RCIM_Cubic, leaveTangent=-0.0405, tangentMode=RCTM_User, time=0.9307, value=1.477}
          {arriveTangent=-8.0089, interpMode=RCIM_Cubic, leaveTangent=-8.0089, tangentMode=RCTM_User, time=1, value=0}
         ]}}})
    ))
   &ScaleSpriteSize(Uniform Curve Sprite Scale=DataInterface() {propertyValues {Curve {keys [
        {tangentMode=RCTM_User, time=0, value=1}
        <at 0.9036 1>
        {arriveTangent=-0.0405, interpMode=RCIM_Cubic, leaveTangent=-0.0405, tangentMode=RCTM_User, time=0.9307, value=1.4739}
        {arriveTangent=-8.0089, interpMode=RCIM_Cubic, leaveTangent=-8.0089, tangentMode=RCTM_User, time=1, value=0}
       ]}}})
   &Drag(
    Drag=&FloatFromCurve(
     FloatCurve=DataInterface() {propertyValues {Curve {keys [{interpMode=RCIM_Cubic, tangentMode=RCTM_User, time=0, value=0}, {interpMode=RCIM_Cubic, tangentMode=RCTM_User, time=0.4807, value=1}]}}}
     Scale Curve=&RandomRangeFloat(Minimum=2, Maximum=1.7)
    )
    Use Linear Drag=true
    Use Rotational Drag=true
   )
   &SolveForcesAndVelocity(Clamp Velocity=false, Limit Acceleration=false)
  ]
  rendererClasses [&NiagaraMeshRendererProperties, &NiagaraSpriteRendererProperties]
  renderers [
   {rendererClass=&NiagaraMeshRendererProperties, rendererIndex=0}
   {rendererClass=&NiagaraSpriteRendererProperties, rendererIndex=1}
  ]
 }
 Confetti_Packed {
  simTarget=GPUComputeSim
  emitterUpdateScript [
   &EmitterState(Life Cycle Mode=::NewEnumerator1, Loop Behavior=::NewEnumerator1, Loop Duration=4, UseLoopDelay=false)
   &SpawnBurst_Instantaneous(Spawn Count=40, Use Spawn Probability=false)
   &SpawnBurst_Instantaneous(Spawn Count=5, Spawn Group=1, Use Spawn Probability=false) {module=SpawnBurst_Instantaneous001}
   &SpawnBurst_Instantaneous(Spawn Count=25, Spawn Group=2, Use Spawn Probability=false) {module=SpawnBurst_Instantaneous002}
  ]
  particleSpawnScript [
   /SetVariables(Particles.SpawnGroup=$Emitter.SpawnGroup) {isSetParameters=true}
   /SetVariables(Particles.VisibilityTag=$Particles.SpawnGroup) {isSetParameters=true}
   &InitializeParticle(
    Color=&RandomRangeLinearColor(Minimum=<rgba 0.9896 1 0.9792 1>, Maximum=<rgba 0.013 0.3569 0.75 1>)
    Lifetime Max=1.3
    Lifetime Min=0.9
    Lifetime Mode=::NewEnumerator1
    Sprite Rotation Mode=::NewEnumerator2
    Sprite Size Mode=::NewEnumerator1
    Uniform Sprite Size Max=2
    Uniform Sprite Size Min=4
    UsePositionOffset=false
    Write Material Random=true
   )
   /SetVariables(
    Particles.Sprite_01_Color:LinearColor=&MakeCustomLinearColorFromBool(
     Bool=&RandomBool()
     True Color=&RandomRangeLinearColor(Minimum=<rgba 0 0.305 0 1>, Maximum=<rgba 0.4 0.305 0 1>)
     False Color=&RandomRangeLinearColor(Minimum=<rgba 0.85 0.305 0 1>, Maximum=<rgba 1 0.305 0 1>)
    )
    Particles.Sprite_01_Size:Vector2f=&Vector2DFromFloat(Value=&RandomRangeFloat(Minimum=12, Maximum=25))
   ) {isSetParameters=true}
   /SetVariables(
    Particles.Sprite_02_Color:LinearColor=&HueShiftLinearColor(
     Hue=0
     Saturation=&RandomRangeFloat(Minimum=2, Maximum=8)
     Value=&RandomRangeFloat(Minimum=1, Maximum=10)
     Linear Color=&RandomRangeLinearColor(Minimum=<rgba 0 0 0 1>, Maximum=<rgba 1 1 1 1>, Color Channel Mode=::NewEnumerator2)
    )
    Particles.Sprite_02_Size:Vector2f=&Vector2DFromFloat(Value=&RandomRangeFloat(Minimum=3, Maximum=8))
   ) {isSetParameters=true}
   &ShapeLocation(
    MaskSpawnGroup=false
    Override Seed=false
    Sphere Radius=/Niagara/DynamicInputs/Math/LerpMultipleFloats(
     Float Count:ENiagara_MultipleLerpCount=::NewEnumerator0
     1:NiagaraFloat=&RandomRangeFloat(Minimum=5, Maximum=15)
     2:NiagaraFloat=&RandomRangeFloat(Minimum=5, Maximum=15)
     3:NiagaraFloat=&RandomRangeFloat(Minimum=2, Maximum=5)
     4:NiagaraFloat=?
     5:NiagaraFloat=?
     Alpha:NiagaraFloat=&Divide_Float(
      A=&MakeFloatFromInt(Int=3)
      B=&MakeFloatFromInt(Int=$Particles.SpawnGroup)
     )
    )
    UseOwnerScale=true
   )
   &AddVelocity(
    Module.Use Falloff Distance=false
    Override Seed=false
    Velocity Mode=::NewEnumerator1
    Velocity Speed=&LerpMultipleFloats(
     Float Count:ENiagara_MultipleLerpCount=::NewEnumerator0
     1:NiagaraFloat=&RandomRangeFloat(Minimum=200, Maximum=400)
     2:NiagaraFloat=&RandomRangeFloat(Minimum=250, Maximum=400)
     3:NiagaraFloat=&RandomRangeFloat(Minimum=200, Maximum=400)
     4:NiagaraFloat=?
     5:NiagaraFloat=?
     Alpha:NiagaraFloat=&Divide_Float(
      A=&MakeFloatFromInt(Int=3)
      B=&MakeFloatFromInt(Int=$Particles.SpawnGroup)
     )
    )
   )
   &PointForce(
    Force Strength=&RandomRangeFloat(Minimum=100, Maximum=300)
    Module.Use Falloff=false
    Module.UseLinearForce=false
   ) {enabled=false}
   &PointForce(
    Force Origin=&SubtractVectorFromPosition(Position=&SimulationPosition(), Vector=<0 0 0>)
    Force Strength=&RandomRangeFloat(Minimum=500, Maximum=300)
    Linear Force Vector:Vector3f=&TransformVector(Vector=&RandomRangeVector(Minimum=<-1 0.1 0.3>, Maximum=<1 1 1>), Destination Space=::World)
    Module.Use Falloff=false
    Module.UseLinearForce=true
   ) {module=PointForce001}
   &ApplyInitialForces()
  ]
  particleUpdateScript [
   &ParticleState()
   &Drag(
    Drag=&LerpMultipleFloats(
     Float Count:ENiagara_MultipleLerpCount=::NewEnumerator0
     1:NiagaraFloat=2
     2:NiagaraFloat=5
     3:NiagaraFloat=6
     4:NiagaraFloat=?
     5:NiagaraFloat=?
     Alpha:NiagaraFloat=&Divide_Float(
      A=&MakeFloatFromInt(Int=3)
      B=&MakeFloatFromInt(Int=$Particles.SpawnGroup)
     )
    )
    Use Linear Drag=true
    Use Rotational Drag=false
   )
   &CurlNoiseForce(Bias Noise=false, Noise Strength=295.2648, Pan Noise=false)
   &GravityForce(Gravity=/Niagara/DynamicInputs/Math/LerpMultipleVectors(
     Alpha:NiagaraFloat=&Divide_Float(
      A=&MakeFloatFromInt(Int=3)
      B=&MakeFloatFromInt(Int=$Particles.SpawnGroup)
     )
     V 1:Vector3f=&RandomRangeVector(Minimum=<0 0 -360>, Maximum=<1 1 -151.1983>)
     V 2:Vector3f=&RandomRangeVector(Minimum=<0 0 -360>, Maximum=<1 1 -171.3172>)
     V 3:Vector3f=&RandomRangeVector(Minimum=<0 0 -100>, Maximum=<1 1 -229.7901>)
     V 4:Vector3f=?
     V 5:Vector3f=?
     Vector Count:ENiagara_MultipleLerpCount=::NewEnumerator0
    ))
   &SolveForcesAndVelocity(Clamp Velocity=false, Limit Acceleration=false)
   &ScaleSpriteSize(Scale Factor=&Vector2DFromFloat(Value=&FloatFromCurve(
      FloatCurve=DataInterface() {propertyValues {Curve {keys [
          {arriveTangent=-0.4596, interpMode=RCIM_Cubic, leaveTangent=-0.4596, tangentMode=RCTM_User, time=0.55, value=1}
          {arriveTangent=-4.9031, interpMode=RCIM_Cubic, leaveTangent=-4.9031, tangentMode=RCTM_User, time=1, value=0}
         ]}}}
      CurveIndex=&Divide_Float(
       A=$Particles.NormalizedAge
       B=&RandomRangeFloat(Minimum=0.2, Maximum=1, Randomness Mode=::NewEnumerator2, Random Seed=?)
      )
     )))
   &ScaleColor(
    Scale RGB=&MakeVector(
     X=1
     Y=&Add_Float(
      A=&RandomRangeFloat(Minimum=0, Maximum=1000)
      B=&Lerp_Float(
       A=0
       B=&RandomRangeFloat(Minimum=25, Maximum=150)
       Alpha=&Divide_Float(
        A=$Particles.NormalizedAge
        B=&RandomRangeFloat(Minimum=0.2, Maximum=1, Randomness Mode=::NewEnumerator2, Random Seed=?)
       )
      )
     )
     Z=1
    )
    ScaleA=false
    ScaleRGB=true
   )
   /SetVariables(
    Particles.Sprite_01_Color:LinearColor=&MultiplyLinearColor(Linear Color=$Particles.Initial.Sprite_01_Color, Scale Factor=<1 1 1 1>)
    Particles.Sprite_01_Size:Vector2f=&Multiply_Vector2D(
     A=$Particles.Initial.Sprite_01_Size
     B=&Vector2DFromFloat(Value=&FloatFromCurve(
       FloatCurve=DataInterface() {propertyValues {Curve {keys [
           {arriveTangent=-0.4596, interpMode=RCIM_Cubic, leaveTangent=-0.4596, tangentMode=RCTM_User, time=0.55, value=1}
           {arriveTangent=-4.9031, interpMode=RCIM_Cubic, leaveTangent=-4.9031, tangentMode=RCTM_User, time=1, value=0}
          ]}}}
       CurveIndex=&Divide_Float(
        A=$Particles.NormalizedAge
        B=&RandomRangeFloat(Minimum=0.5, Maximum=1, Randomness Mode=::NewEnumerator2, Random Seed=?)
       )
       Scale Curve=0.8
      ))
    )
   ) {isSetParameters=true}
   /SetVariables(
    Particles.Sprite_02_Color:LinearColor=&MultiplyLinearColor(
     Linear Color=$Particles.Initial.Sprite_02_Color
     Scale Factor=&MakeVector4_FromLinearColor(LinearColor=&MultiplyLinearColor(
       Linear Color=&FN_Select_Color_from_Time_Of_Day(
        Use Manual Time of Day Thresholds=true
        Morning=<rgba 1 0.75 0.7753 1>
        Day=<rgba 1 0.8842 0.8294 1>
        Evening=<rgba 1 0.625 0.5 1>
        Night=<rgba 0.7 0.3494 0.4483 1>
        Evening Threshold=21
       )
       Scale Factor=&MakeFloat4FromVectorAndFloat(
        "Vector (RGB)"=<1 1 1>
        "Float (Alpha)"=&Lerp_Float(
         A=1
         B=0
         Alpha=&Divide_Float(
          A=$Particles.NormalizedAge
          B=&RandomRangeFloat(Minimum=0.2, Maximum=1, Randomness Mode=::NewEnumerator2, Random Seed=?)
         )
        )
       )
      ))
    )
    Particles.Sprite_02_Size:Vector2f=&Multiply_Vector2D(
     A=$Particles.Initial.Sprite_02_Size
     B=&Vector2DFromFloat(Value=&FloatFromCurve(
       FloatCurve=DataInterface() {propertyValues {Curve {keys [<at 0.0018 0.1957>,<at 0.0527 0.819>,<at 0.1169 0.6555>,<at 0.1748 0.9015>,<at 0.2394 0.582>,<at 0.332 1.0023>,<at 0.4947 0.3707>,<at 0.6727 1>,<at 1 0>]}}}
       CurveIndex=&Divide_Float(
        A=$Particles.NormalizedAge
        B=&RandomRangeFloat(Minimum=0.2, Maximum=1, Randomness Mode=::NewEnumerator2, Random Seed=?)
       )
      ))
    )
   ) {isSetParameters=true}
  ]
  rendererClasses [&NiagaraSpriteRendererProperties]
  renderers [
   {rendererClass=&NiagaraSpriteRendererProperties}
   {rendererClass=&NiagaraSpriteRendererProperties, rendererIndex=1}
   {rendererClass=&NiagaraSpriteRendererProperties, rendererIndex=2}
  ]
 }
 Flame_Streaks {
  emitterUpdateScript [
   &EmitterState()
   &SpawnBurst_Instantaneous(
    Spawn Count=&RandomRangeInt(Minimum=6, Maximum=9)
    Spawn Time=0.01
    Use Spawn Probability=false
    UseLoopCountLimit=false
   )
  ]
  particleSpawnScript [
   &InitializeParticle(
    Lifetime=&RandomRangeFloat(Minimum=0.12, Maximum=0.17)
    Sprite Size=&Multiply_Vector2D_ByFloat(
     Float=&RandomRangeFloat(Minimum=0.5, Maximum=1.5)
     Vector2D=&RandomRangeVector2D(Minimum=<8 33>, Maximum=<13 40>)
    )
    Sprite Size Mode=::NewEnumerator4
    UsePositionOffset=false
    Write Material Random=true
   )
   &SubUVAnimation(
    End Frame Range Override=3
    Evaluation Type=::NewEnumerator0
    "Sprite Renderer "=DataInterface() {propertyValues {SpriteRenderer=&self:Flame_Streaks_1.NiagaraSpriteRendererProperties_0}}
    SubUV Animation Mode:ENiagara_SubUVLookupModeV2=::NewEnumerator2
    UseEndFrame=true
    UseStartFrame=true
   )
   &ShapeLocation(
    Handle Radius=1
    Large Radius=15.6226
    MaskSpawnGroup=false
    Offset=<0 0 10>
    Offset Mode=::NewEnumerator0
    Override Seed=false
    Shape Primitive=::NewEnumerator3
    UseOwnerScale=true
   )
   &AddVelocity(
    Module.Use Falloff Distance=false
    Origin Offset=&MakeVector(X=0, Y=0, Z=&RandomRangeFloat(Minimum=0, Maximum=-15))
    Override Seed=false
    Velocity Mode=::NewEnumerator1
    Velocity Speed=&RandomRangeFloat(Minimum=300, Maximum=200)
   )
   &CameraOffset(Camera Offset Amount=30)
  ]
  particleUpdateScript [
   &Drag(
    Drag=&Multiply_Float(
     A=&RandomRangeFloat(Minimum=10, Maximum=20)
     B=&FloatFromCurve(FloatCurve=DataInterface() {propertyValues {Curve {keys [<at 0 0>,<at 0.3 1>]}}})
    )
    Use Linear Drag=true
    Use Rotational Drag=true
   )
   &SolveForcesAndVelocity(Clamp Velocity=false, Limit Acceleration=false)
   &ParticleState()
   &ScaleSpriteSize(
    Non-Uniform Curve Sprite Scale=DataInterface() {propertyValues {
      XCurve {keys [{interpMode=RCIM_Cubic, tangentMode=RCTM_User, time=0, value=1}, {interpMode=RCIM_Cubic, tangentMode=RCTM_User, time=1, value=1}]}
      YCurve {keys [{interpMode=RCIM_Cubic, tangentMode=RCTM_User, time=0.35, value=1.3}, {interpMode=RCIM_Cubic, tangentMode=RCTM_User, time=1, value=0.7}]}
     }}
    Scale Sprite Size Mode=::NewEnumerator3
   )
   &ScaleColor(
    Scale Alpha=&FloatFromCurve(FloatCurve=DataInterface() {propertyValues {Curve {keys [<at 0 1>,<at 1 0>]}}})
    Scale RGB=&MakeVectorFromLinearColorRGB(Color=&ColorFromCurve(ColorCurve=DataInterface() {propertyValues {
        AlphaCurve {keys [<at 0 1 RCIM_Cubic>]}
        BlueCurve {keys [<at 0.0228 1.61 RCIM_Cubic>,<at 0.2447 0.0037>,<at 0.3924 0.8995>,<at 0.5337 0.9561>,<at 0.7222 0.1807>,<at 1 0 RCIM_Cubic>]}
        GreenCurve {keys [<at 0.0228 4.7542 RCIM_Cubic>,<at 0.2447 0.8205>,<at 0.3924 0.097>,<at 0.5337 0.7895>,<at 0.7222 2.1143>,<at 1 0.2807 RCIM_Cubic>]}
        RedCurve {keys [<at 0.0228 7 RCIM_Cubic>,<at 0.2447 3>,<at 0.3924 1.0407>,<at 0.5337 0.0255>,<at 0.7222 0>,<at 1 1 RCIM_Cubic>]}
       }}))
    ScaleA=true
    ScaleRGB=true
   )
  ]
  rendererClasses [&NiagaraSpriteRendererProperties]
  renderers [{rendererClass=&NiagaraSpriteRendererProperties}]
 }
 Flash001 {
  emitterUpdateScript [
   &EmitterState(Life Cycle Mode=::NewEnumerator1, Loop Behavior=::NewEnumerator1, Loop Duration=1, UseLoopDelay=false)
   &SpawnBurst_Instantaneous(Spawn Count=1, Use Spawn Probability=false)
  ]
  particleSpawnScript [
   &InitializeParticle(
    Lifetime=0.1
    Mesh Scale=&Multiply_VectorByFloat(Vector=&RandomRangeVector(Minimum=<1 2 2>, Maximum=<0.75 1.5 1.5>), Float=0.7)
    Mesh Scale Mode=::NewEnumerator4
    UsePositionOffset=true
    Write Material Random=true
   )
   &DynamicMaterialParameters(
    Index 0 Param 2=45
    Index 0 Param 3=1
    Index 0 Param 4=1
    Module.Param0WriteEnabled=false
    Module.Param1WriteEnabled=true
    Module.Param2WriteEnabled=true
    Module.Param3WriteEnabled=true
   ) {module=DynamicMaterialParameters001}
   &InitialMeshOrientation(
    Rotation=&RandomRangeVector(Minimum=<0 0 0>, Maximum=<1 1 1>)
    Use Orientation Vector=true
    Use Rotation Vector=true
   )
   &CameraOffset(Camera Offset Amount=15)
  ]
  particleUpdateScript [
   &ParticleState()
   &DynamicMaterialParameters(
    Index 0 Param 1=&FloatFromCurve(FloatCurve=DataInterface() {propertyValues {Curve {keys [<at 0.56 0>,<at 1 1>]}}})
    Module.Param0WriteEnabled=true
    Module.Param1WriteEnabled=false
    Module.Param2WriteEnabled=false
    Module.Param3WriteEnabled=false
   )
   &ScaleMeshSize(Scale Factor=&VectorFromFloat(Value=&FloatFromCurve(FloatCurve=DataInterface() {propertyValues {Curve {keys [<at 0 0.5>,<at 1 1>]}}})))
   &ScaleColor(
    Scale Alpha=&FloatFromCurve(FloatCurve=DataInterface() {propertyValues {Curve {keys [<at 0 1>,<at 1 0>]}}})
    Scale RGB=&MakeVectorFromLinearColorRGB(Color=&ColorFromCurve(ColorCurve=DataInterface() {propertyValues {
        AlphaCurve {keys []}
        BlueCurve {keys [<at 0 0.2917 RCIM_Cubic>,<at 0.15 0.3906>,<at 0.833 0.2031 RCIM_Cubic>]}
        GreenCurve {keys [<at 0 0.7049 RCIM_Cubic>,<at 0.15 0.6445>,<at 0.833 0.4688 RCIM_Cubic>]}
        RedCurve {keys [<at 0 1 RCIM_Cubic>,<at 0.15 1>,<at 0.833 1 RCIM_Cubic>]}
       }}))
    ScaleA=true
    ScaleRGB=true
   )
   &ScaleColor(
    Scale RGB=&VectorFromFloat(Value=&FloatFromCurve(FloatCurve=DataInterface() {propertyValues {Curve {keys [<at 0 10>,<at 0.2 3>,<at 1 1>]}}}))
    ScaleA=true
    ScaleRGB=true
   ) {module=ScaleColor001}
   &ScaleColor(
    Scale RGB=&VectorFromFloat(Value=&FNTimeOfDayScalar(Night Scale=0.5, Morning Scale=0.7, Evening Scale=0.7))
    ScaleA=true
    ScaleRGB=true
   ) {module=ScaleColor002}
   &self:ColorFromTOD_12(
    Afternoon:LinearColor=<rgba 1 0.8842 0.8294 1>
    Evening:LinearColor=<rgba 1 0.625 0.5 1>
    Morning:LinearColor=<rgba 1 0.75 0.7753 1>
    Night:LinearColor=<rgba 0.7 0.3494 0.4483 1>
    TOD:Vector4f=&MakeVector4_FromLinearColor(LinearColor=$NPC.FortniteNPC.FortniteActiveTimeOfDay)
   ) {module=ColorFromTOD_12}
  ]
  rendererClasses [&NiagaraMeshRendererProperties]
  renderers [{rendererClass=&NiagaraMeshRendererProperties, rendererIndex=0}]
 }
 Impact_Flash {
  emitterUpdateScript [
   &EmitterState()
   &SpawnBurst_Instantaneous(Use Spawn Probability=false, UseLoopCountLimit=false)
  ]
  particleSpawnScript [&InitializeParticle(
    Color=&SelectLinearColorFromArray(
     Array Sampling Mode=::NewEnumerator0
     Direct Array Index=?
     Color Selection Array=DataInterface() {propertyValues {ColorData [<rgba 1 0.3385 0.0109 1>, <rgba 0.1002 1 0.5472 1>, <rgba 0.0124 0.2651 1 1>, <rgba 1 0.0676 0.5257 1>]}}
     Random Seed=0
    )
    Lifetime=0.15
    Position Offset=<0 0 8>
    Sprite Size Mode=::NewEnumerator3
    Uniform Sprite Size=109.9378
    UsePositionOffset=true
    Write Material Random=true
   )]
  particleUpdateScript [
   &CameraOffset(Camera Offset Amount=&FloatFromCurve(
     FloatCurve=DataInterface() {propertyValues {Curve {keys [
         {interpMode=RCIM_Cubic, tangentMode=RCTM_User, time=0, value=1}
         {arriveTangent=-0.5583, interpMode=RCIM_Cubic, leaveTangent=-0.5583, tangentMode=RCTM_User, time=0.25, value=0.3}
        ]}}}
     Scale Curve=30
    ))
   &ParticleState()
   &ScaleSpriteSize(Uniform Curve Sprite Scale=DataInterface() {propertyValues {Curve {keys [
        {arriveTangent=2.4123, interpMode=RCIM_Cubic, leaveTangent=2.4123, tangentMode=RCTM_User, time=0, value=0.3}
        {interpMode=RCIM_Cubic, tangentMode=RCTM_User, time=0.9, value=1}
       ]}}})
   &ScaleColor(
    Scale Alpha=&FloatFromCurve(
     FloatCurve=DataInterface() {propertyValues {Curve {keys [
         {interpMode=RCIM_Cubic, tangentMode=RCTM_User, time=0, value=0}
         {arriveTangent=-1.8484, interpMode=RCIM_Cubic, leaveTangent=-1.8484, tangentMode=RCTM_User, time=0.1, value=1}
         {interpMode=RCIM_Cubic, tangentMode=RCTM_User, time=1, value=0}
        ]}}}
     Scale Curve=&FNTimeOfDayScalar(Night Scale=0.12, Morning Scale=0.12, Day Scale=0.15, Evening Scale=0.12)
    )
    Scale RGB=&MakeVectorFromLinearColorRGB(Color=&ColorFromCurve(ColorCurve=DataInterface() {propertyValues {
        AlphaCurve {keys [<at 0 1 RCIM_Cubic>]}
        BlueCurve {keys [<at 0.0598 2 RCIM_Cubic>,<at 0.5405 1 RCIM_Cubic>]}
        GreenCurve {keys [<at 0.0598 2 RCIM_Cubic>,<at 0.5405 1 RCIM_Cubic>]}
        RedCurve {keys [<at 0.0598 2 RCIM_Cubic>,<at 0.5405 1 RCIM_Cubic>]}
       }}))
    ScaleA=true
    ScaleRGB=true
   )
   /SetVariables(Particles.ColorFlare:LinearColor=&MultiplyLinearColor(
     Linear Color=$Particles.Color
     Scale Factor=/Niagara/DynamicInputs/Multiply/Multiply_Vector4_ByFloat(
      Vector4:Vector4f=<10 10 10 10>
      Float:NiagaraFloat=&FNTimeOfDayScalar(Night Scale=0.8, Morning Scale=0.8, Day Scale=0.8, Evening Scale=0.6)
     )
    )) {isSetParameters=true}
  ]
  rendererClasses [&NiagaraSpriteRendererProperties, &NiagaraLightRendererProperties]
  renderers [
   {rendererClass=&NiagaraSpriteRendererProperties}
   {rendererClass=&NiagaraSpriteRendererProperties, rendererIndex=1}
   {rendererClass=&NiagaraLightRendererProperties, rendererIndex=2}
  ]
 }
 MeshRipples {
  emitterUpdateScript [
   &EmitterState(Inactive Response=::NewEnumerator1, Life Cycle Mode=::NewEnumerator1, Loop Behavior=::NewEnumerator1, Loop Duration=1, UseLoopDelay=false)
   &SpawnBurst_Instantaneous(Spawn Time=0.2, Use Spawn Probability=false, UseLoopCountLimit=false) {enabled=false, module=SpawnBurst_Instantaneous002}
   &SpawnBurst_Instantaneous(Spawn Time=0.05, Use Spawn Probability=false, UseLoopCountLimit=false) {enabled=false, module=SpawnBurst_Instantaneous001}
   &SpawnBurst_Instantaneous(Use Spawn Probability=false, UseLoopCountLimit=false)
  ]
  particleSpawnScript [
   &InitializeParticle(
    Color=&FNTimeOfDayColor(
     Night Color=<rgba 0.1584 0.3819 0.8714 0.1>
     Morning Color=<rgba 0.7679 0.6226 0.6489 0.05>
     Day Color=<rgba 0.8714 0.8059 0.6405 0.1>
     Evening Color=<rgba 0.3669 0.6418 0.8714 0.05>
    )
    Lifetime=&FloatFromCurve(
     FloatCurve=DataInterface() {propertyValues {Curve {keys [{interpMode=RCIM_Cubic, tangentMode=RCTM_User, time=0, value=0.2}, {interpMode=RCIM_Cubic, tangentMode=RCTM_User, time=0.7637, value=1}]}}}
     CurveIndex=$Emitter.Age
    )
    Mesh Scale=&Multiply_VectorByFloat(
     Vector=&MakeVector(X=1.4027, Y=0.75, Z=&RandomRangeFloat(Minimum=3.2589, Maximum=1.5))
     Float=0.744
    )
    Mesh Scale Mode=::NewEnumerator4
    Position Offset=<0 0 -29>
    UsePositionOffset=true
    Write Material Random=true
   )
   &AddVelocity(Velocity=&RandomRangeVector(Minimum=<0 0 -10>, Maximum=<0 0 -20>), Velocity Speed Scale=2)
   &InitialMeshOrientation(
    Mesh Axis To Orient=<0 0 -1>
    Mesh Orientation Mode=::NewEnumerator2
    Orientation Vector=<0 0 1>
    Rotation=&RandomRangeVector(Minimum=<0 0 0.542>, Maximum=<0 0 0.5044>)
    Use Rotation Vector=true
   )
  ]
  particleUpdateScript [
   &ParticleState()
   &Drag(Drag=5, Use Linear Drag=true, Use Rotational Drag=true)
   &SolveForcesAndVelocity(Clamp Velocity=false, Limit Acceleration=false)
   &ScaleMeshSize(Scale Factor=&VectorFromFloat(Value=&FloatFromCurve(FloatCurve=DataInterface() {propertyValues {Curve {keys [{interpMode=RCIM_Cubic, tangentMode=RCTM_User, time=0, value=0.3619}, {interpMode=RCIM_Cubic, tangentMode=RCTM_User, time=1, value=0.6}]}}})))
   &DynamicMaterialParameters(
    Index 0 Param 1=&FloatFromCurve(FloatCurve=DataInterface() {propertyValues {Curve {keys [<at 0.1 5 RCIM_Cubic>,<at 0.95 0 RCIM_Cubic>]}}})
    Index 0 Param 2=&FloatFromCurve(FloatCurve=DataInterface() {propertyValues {Curve {keys [<at 0 5 RCIM_Cubic>,<at 0.4 19.4015 RCIM_Cubic>]}}})
    Index 0 Param 3=1
    Index 0 Param 4=&FloatFromCurve(FloatCurve=DataInterface() {propertyValues {Curve {keys [<at 0 0.9407 RCIM_Cubic>,<at 1 0 RCIM_Cubic>]}}})
    Module.Index0Float0Write=true
    Module.Index0Float1Write=true
    Module.Index0Float2Write=true
    Module.Index0Float3Write=true
   )
   &ScaleColor(
    Scale Alpha=&FloatFromCurve(FloatCurve=DataInterface() {propertyValues {Curve {keys [
         {arriveTangent=9.0725, interpMode=RCIM_Cubic, leaveTangent=9.0725, tangentMode=RCTM_User, time=0, value=0}
         {arriveTangent=-2.5249, interpMode=RCIM_Cubic, leaveTangent=-2.5249, tangentMode=RCTM_User, time=0.1408, value=0.9847}
         {arriveTangent=-0.0405, interpMode=RCIM_Cubic, leaveTangent=-0.0405, tangentMode=RCTM_User, time=1, value=0}
        ]}}})
    ScaleA=true
    ScaleRGB=true
   )
  ]
  rendererClasses [&NiagaraMeshRendererProperties]
  renderers [{rendererClass=&NiagaraMeshRendererProperties, rendererIndex=0}]
 }
 SpikeStreaksParticle {
  emitterUpdateScript [
   &EmitterState(
    Enable Distance Culling=true
    Life Cycle Mode=::NewEnumerator1
    Loop Behavior=::NewEnumerator1
    Loop Duration=1
    Max Distance Response=::KillImmediately
    MaxDistance=3500
    Module.bUseMaxDistance=true
    Module.bUseMinDistance=false
    Scalability Mode=::NewEnumerator1
    UseLoopDelay=false
   )
   &SpawnBurst_Instantaneous(Spawn Count=&RandomRangeInt(Minimum=6, Maximum=8), Use Spawn Probability=false)
  ]
  particleSpawnScript [
   &InitializeParticle(
    Color Maximum=<rgba 0.1451 1 0.3307 1>
    Color Minimum=<rgba 1 0.0464 0.146 1>
    Color Mode=::NewEnumerator2
    Lifetime Max=0.15
    Lifetime Min=0.17
    Lifetime Mode=::NewEnumerator1
    Mesh Scale Mode=::NewEnumerator3
    Mesh Uniform Scale=0.6
    Sprite Size Max=<10 50>
    Sprite Size Min=<15 75>
    Sprite Size Mode=::NewEnumerator2
    UsePositionOffset=true
    Write Material Random=true
   )
   &AddVelocityFromPoint(Module.Use Falloff Distance=false)
  ]
  particleUpdateScript [
   &ParticleState()
   &SolveForcesAndVelocity(Clamp Velocity=false, Limit Acceleration=false)
   &ScaleSpriteSize(Scale Factor=&Vector2DFromFloat(Value=&FloatFromCurve(FloatCurve=DataInterface() {propertyValues {Curve {keys [<at 0 0.6>,<at 1 2>]}}}, Scale Curve=0.5)))
   &ScaleColor(
    Scale Alpha=&FloatFromCurve(FloatCurve=DataInterface() {propertyValues {Curve {keys [<at 0.5919 1>,<at 1 0>]}}}, Scale Curve=1.5)
    Scale RGB=&VectorFromFloat(Value=&FNTimeOfDayScalar(Night Scale=0.3, Morning Scale=0.5, Evening Scale=0.5))
    ScaleA=true
    ScaleRGB=true
   )
   &CameraOffset(Camera Offset Amount=10) {module=CameraOffset001}
  ]
  rendererClasses [&NiagaraSpriteRendererProperties]
  renderers [{rendererClass=&NiagaraSpriteRendererProperties}]
 }
}
summary {
 systemName=NS_Emotes_SoapPocketPunch_Candy
 emitters [
  {
   emitterName=Candy
   simTarget=GPUComputeSim
   rendererClasses [&NiagaraMeshRendererProperties, &NiagaraSpriteRendererProperties]
  }
  {emitterName=Confetti_Packed, simTarget=GPUComputeSim, rendererClasses [&NiagaraSpriteRendererProperties]}
  {emitterName=SpikeStreaksParticle, rendererClasses [&NiagaraSpriteRendererProperties]}
  {emitterName=Flash001, rendererClasses [&NiagaraMeshRendererProperties]}
  {emitterName=MeshRipples, rendererClasses [&NiagaraMeshRendererProperties]}
  {emitterName=Flame_Streaks, rendererClasses [&NiagaraSpriteRendererProperties]}
  {emitterName=Impact_Flash, rendererClasses [&NiagaraSpriteRendererProperties, &NiagaraLightRendererProperties]}
 ]
 userVariables [
  User.bIsFrontEnd {type=NiagaraBool}
  User.bIsFrontEndPreview {type=NiagaraBool}
 ]
}
